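# Work-in-progress UI material function for the Creative HUD that draws a polygon from six input vertices, despite its "4 Vertices" description.
# A Custom HLSL node computes the polygon's signed distance over UVs remapped to a centered -1 to 1 space.
# The absolute distance minus Outline Thickness forms an outline SDF.
# It outputs both raw SDFs plus Fill and Outline masks from inverted SmoothSteps, with Glow Min/Max inputs for falloff.
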MaterialFunction /Game/Athena/HUD/Creative/MaterialFuntions/MF_UI_WIP_Creative_Polygon_6_Vertex
graph {
 properties {description=Independant 4 Vertices Function, libraryCategoriesText [UI_Shader_Library]}
 nodes [
  FunctionOutput(None=NamedRerouteUsage()) {outputName=SDF, sortPriority=3}
  FunctionOutput(None=NamedRerouteUsage()) {outputName=SDF Outline, sortPriority=4}
  n1 = NamedRerouteUsage()
  n2 = NamedRerouteUsage()
  n3 = NamedRerouteUsage()
  FunctionOutput(None=OneMinus(None=SmoothStep(Min=@n1, Max=@n2, Value=@n3))) {outputName=Fill}
  n4 = NamedRerouteUsage()
  n5 = NamedRerouteUsage()
  n6 = NamedRerouteUsage()
  FunctionOutput(None=OneMinus(None=SmoothStep(Min=@n4, Max=@n5, Value=@n6))) {outputName=Outline, sortPriority=2}
  n7 = FunctionInput(Preview=Constant()) {
   bUsePreviewValueAsDefault=true
   description=Determines glow falloff strength. Can not go below GlowMin value or results will invert
   inputName=Glow Min
   inputType=FunctionInput_Scalar
   sortPriority=8
  }
  NamedRerouteDeclaration(None=@n7) {name=Glow Min, nodeColor=<rgba 0 0 0 1>}
  n8 = FunctionInput(Preview=Constant()) {
   bUsePreviewValueAsDefault=true
   description=Determines glow falloff strength. Can not go below GlowMin value or results will invert
   inputName=Glow Max
   inputType=FunctionInput_Scalar
   sortPriority=9
  }
  NamedRerouteDeclaration(None=@n8) {name=Glow Max, nodeColor=<rgba 0 0 0 1>}
  n9 = FunctionInput(Preview=TextureCoordinate()) {
   bUsePreviewValueAsDefault=true
   description=The Position Coordinates to measure the box. Default is World Position and can be left default in most cases.
   inputName=UVs
   inputType=FunctionInput_Vector2
  }
  MaterialExpressionConstantBiasScale_0 = ConstantBiasScale(None=@n9) {bias=-0.5, scale=2}
  MaterialExpressionFunctionInput_8 = FunctionInput(Preview=Constant2Vector() {g=-1}) {bUsePreviewValueAsDefault=true, inputName=vertex_0, inputType=FunctionInput_Vector2, sortPriority=1}
  MaterialExpressionFunctionInput_0 = FunctionInput(Preview=Constant2Vector() {r=1}) {bUsePreviewValueAsDefault=true, inputName=vertex_1, inputType=FunctionInput_Vector2, sortPriority=2}
  MaterialExpressionFunctionInput_3 = FunctionInput(Preview=Constant2Vector() {r=0.5}) {bUsePreviewValueAsDefault=true, inputName=vertex_2, inputType=FunctionInput_Vector2, sortPriority=3}
  MaterialExpressionFunctionInput_4 = FunctionInput(Preview=Constant2Vector() {g=1}) {bUsePreviewValueAsDefault=true, inputName=vertex_3, inputType=FunctionInput_Vector2, sortPriority=4}
  MaterialExpressionFunctionInput_5 = FunctionInput(Preview=Constant2Vector() {r=-0.5}) {bUsePreviewValueAsDefault=true, inputName=vertex_4, inputType=FunctionInput_Vector2, sortPriority=5}
  MaterialExpressionFunctionInput_6 = FunctionInput(Preview=Constant2Vector() {r=-1}) {bUsePreviewValueAsDefault=true, inputName=vertex_5, inputType=FunctionInput_Vector2, sortPriority=6}
  n10 = Custom!(
   p=@MaterialExpressionConstantBiasScale_0
   vertex_0=@MaterialExpressionFunctionInput_8
   vertex_1=@MaterialExpressionFunctionInput_0
   vertex_2=@MaterialExpressionFunctionInput_3
   vertex_3=@MaterialExpressionFunctionInput_4
   vertex_4=@MaterialExpressionFunctionInput_5
   vertex_5=@MaterialExpressionFunctionInput_6
  ) {
   code="const int num = 6;\r\nfloat2 v[num] = { vertex_0, vertex_1, vertex_2, vertex_3, vertex_4, vertex_5 };\r\nfloat d = dot(p-v[0],p-v[0]);\r\nfloat s = 1.0;\r\nfor( int i=0, j=num-1; i<num; j=i, i++ )\r\n{\r\n    float2 e = v[j] - v[i];\r\n    float2 w = p - v[i];\r\n    float2 b = w - e*clamp( dot(w,e)/dot(e,e), 0.0, 1.0 );\r\n    d = min( d, dot(b,b) );\r\n\r\n    bool3 cond = bool3( p.y>=v[i].y, \r\n                        p.y <v[j].y, \r\n                        e.x*w.y>e.y*w.x );\r\n    if( all(cond) || all(!(cond)) ) s=-s;  \r\n}\r\n    \r\nreturn s*sqrt(d);"
   description=Polygon 6 Vertex
   outputType=CMOT_Float1
   inputs [
    {
     inputName=p
     input {expression=&self:MaterialExpressionConstantBiasScale_0}
    }
    {
     inputName=vertex_0
     input {expression=&self:MaterialExpressionFunctionInput_8}
    }
    {
     inputName=vertex_1
     input {expression=&self:MaterialExpressionFunctionInput_0}
    }
    {
     inputName=vertex_2
     input {expression=&self:MaterialExpressionFunctionInput_3}
    }
    {
     inputName=vertex_3
     input {expression=&self:MaterialExpressionFunctionInput_4}
    }
    {
     inputName=vertex_4
     input {expression=&self:MaterialExpressionFunctionInput_5}
    }
    {
     inputName=vertex_5
     input {expression=&self:MaterialExpressionFunctionInput_6}
    }
   ]
  }
  NamedRerouteDeclaration(None=@n10) {name=SDF Shape, nodeColor=<rgba 0 0 0 1>}
  n11 = NamedRerouteUsage()
  n12 = FunctionInput(Preview=Constant() {r=0.05}) {
   bUsePreviewValueAsDefault=true
   description=Determines the thickness of the stroke when using one of the outline outputs.
   inputName=Outline Thickness
   inputType=FunctionInput_Scalar
   sortPriority=7
  }
  NamedRerouteDeclaration(None=Subtract(A=Abs(None=@n11), B=@n12) {constB=0.1}) {name=SDF Outline, nodeColor=<rgba 0 0 0 1>}
 ]
}
