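# UI material function for Creative HUD that draws a rounded box with independently set corner radii, preserving corner aspect ratio when given pixel size.
# A Custom HLSL node computes the signed distance field from re-centered UVs, choosing each quadrant's radius and scaling by pixel-size ratio.
# An outline distance derives from the absolute distance minus thickness.
# It outputs both raw SDFs plus Fill and Outline masks made with inverted SmoothStep, with Glow Min/Max inputs for falloff.
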MaterialFunction /Game/Athena/HUD/Creative/MaterialFuntions/MF_UI_Creative_RoundedBox_KeepAspectRatio
graph {nodes [
  FunctionOutput(None=NamedRerouteUsage()) {outputName=SDF, sortPriority=3}
  FunctionOutput(None=NamedRerouteUsage()) {outputName=SDF Outline, sortPriority=4}
  n1 = NamedRerouteUsage()
  n2 = NamedRerouteUsage()
  n3 = NamedRerouteUsage()
  FunctionOutput(None=OneMinus(None=SmoothStep(Min=@n1, Max=@n2, Value=@n3))) {outputName=Fill}
  n4 = NamedRerouteUsage()
  n5 = NamedRerouteUsage()
  n6 = NamedRerouteUsage()
  FunctionOutput(None=OneMinus(None=SmoothStep(Min=@n4, Max=@n5, Value=@n6))) {outputName=Outline, sortPriority=2}
  n7 = FunctionInput(Preview=Constant()) {
   bUsePreviewValueAsDefault=true
   description=Determines glow falloff strength. Can not go below GlowMin value or results will invert
   inputName=Glow Min
   inputType=FunctionInput_Scalar
   sortPriority=4
  }
  NamedRerouteDeclaration(None=@n7) {name=Glow Min, nodeColor=<rgba 0 0 0 1>}
  n8 = FunctionInput(Preview=Constant()) {
   bUsePreviewValueAsDefault=true
   description=Determines glow falloff strength. Can not go below GlowMin value or results will invert
   inputName=Glow Max
   inputType=FunctionInput_Scalar
   sortPriority=5
  }
  NamedRerouteDeclaration(None=@n8) {name=Glow Max, nodeColor=<rgba 0 0 0 1>}
  n9 = FunctionInput(Preview=TextureCoordinate()) {
   bUsePreviewValueAsDefault=true
   description=The Position Coordinates to measure the box. Default is World Position and can be left default in most cases.
   inputName=UVs
   inputType=FunctionInput_Vector2
  }
  MaterialExpressionConstantBiasScale_0 = ConstantBiasScale(None=@n9) {bias=-0.5, scale=2}
  MaterialExpressionFunctionInput_0 = FunctionInput(Preview=Constant2Vector() {g=1, r=1}) {
   bUsePreviewValueAsDefault=true
   description="plug GetUserInterfaceUV \"Pixel Size\" if you want to keep the rounded corners aspect ratio."
   inputName=Pixel Size
   inputType=FunctionInput_Vector2
   sortPriority=1
  }
  MaterialExpressionFunctionInput_8 = FunctionInput(Preview=Constant2Vector() {g=0.5, r=0.5}) {
   bUsePreviewValueAsDefault=true
   description=The size of the Box. Float2 value between 0 and 1.
   inputName=Box Size
   inputType=FunctionInput_Vector2
   sortPriority=2
  }
  MaterialExpressionFunctionInput_1 = FunctionInput(Preview=Constant4Vector() {constant=<rgba 0.2 0 0.2 0.2>}.RGBA) {
   bUsePreviewValueAsDefault=true
   description=set the 4 corners radius separately
   inputName=Corner Radiuses
   inputType=FunctionInput_Vector4
   sortPriority=3
  }
  n10 = Custom!(
   p=@MaterialExpressionConstantBiasScale_0
   pixelSize=@MaterialExpressionFunctionInput_0
   boxSize=@MaterialExpressionFunctionInput_8
   cornerRadiuses=@MaterialExpressionFunctionInput_1
  ) {
   code="float2 pixelSizeFloored = saturate(floor(pixelSize.x/pixelSize.y));\r\n\r\nfloat2 hRatio = float2((pixelSize.x/pixelSize.y), 1);\r\nfloat2 vRatio = float2(1, (pixelSize.y/pixelSize.x));\r\n\r\nfloat2 ratio = lerp(vRatio, hRatio, pixelSizeFloored);\r\n\r\nfloat2 quadrant = saturate(floor(1-p));\r\nfloat quadX = quadrant.x;\r\nfloat quadY = quadrant.y;\r\n\r\nfloat2 rightCorners = float2(cornerRadiuses.x, cornerRadiuses.y);\r\nfloat2 leftCorners = float2(cornerRadiuses.z, cornerRadiuses.w);\r\n\r\n//Splitting the box vertically, and then horizontally to get 4 quadrants\r\n//Each quadrant value represents the roundness of that corner\r\n\r\nfloat2 verticalSplit = lerp(rightCorners, leftCorners, quadX);\r\nfloat2 cornersRoundness = lerp(verticalSplit.x, verticalSplit.y, quadY);\r\n\r\nfloat2 box = abs(p)-boxSize;\r\nfloat2 cornersCoord = box*ratio + cornersRoundness.x;\r\nreturn min(max(cornersCoord.x,cornersCoord.y),0.0) + length(max(cornersCoord,0.0)) - cornersRoundness.x;"
   description=Rounded Box
   outputType=CMOT_Float1
   inputs [
    {
     inputName=p
     input {expression=&self:MaterialExpressionConstantBiasScale_0}
    }
    {
     inputName=pixelSize
     input {expression=&self:MaterialExpressionFunctionInput_0}
    }
    {
     inputName=boxSize
     input {expression=&self:MaterialExpressionFunctionInput_8}
    }
    {
     inputName=cornerRadiuses
     input {expression=&self:MaterialExpressionFunctionInput_1}
    }
   ]
  }
  NamedRerouteDeclaration(None=@n10) {name=SDF Shape, nodeColor=<rgba 0 0 0 1>}
  n11 = NamedRerouteUsage()
  n12 = FunctionInput(Preview=Constant() {r=0.05}) {
   bUsePreviewValueAsDefault=true
   description=Determines the thickness of the stroke when using one of the outline outputs.
   inputName=Outline Thickness
   inputType=FunctionInput_Scalar
   sortPriority=6
  }
  NamedRerouteDeclaration(None=Subtract(A=Abs(None=@n11), B=@n12) {constB=0.1}) {name=SDF Outline, nodeColor=<rgba 0 0 0 1>}
 ]}
